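# Translucent lit glass material for the cyber glasses on the Sheath Alter Ego player body.
# Mask-driven color and opacity, Fresnel transparency and a baked normal pass through SubUVTextures, then MF_TechArtOverrides.
# A U gradient times time-seeded, pixelated hash noise, offset by the larger of GoggleTransition and FrontEnd_Idle_Bypass, forms a hard mask gating emissive.
# The Leading Edge Color ring branch does not feed the final attributes.
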Material /Game/Characters/Player/Female/Medium/Bodies/F_Med_Sheath/Materials/Alter_Ego/M_Sheath_CyberGlasses
graph {
 properties {
  bUseMaterialAttributes=true
  bUsedWithMeshDeformer=true
  bUsedWithSkeletalMesh=true
  blendMode=BLEND_Translucent
  pixelDepthOffsetMode=PDOM_Legacy
  translucencyLightingMode=TLM_SurfacePerPixelLighting
  usedShadingModels=Default Lit
 }
 nodes [
  n1 = TextureSample() {
   samplerType=SAMPLERTYPE_Masks
   texture=/Game/Characters/Player/Female/Medium/Bodies/F_Med_Sheath/Textures/Alter_Ego/T_F_MED_Sheath_Alter_Ego_M_glass_OP
  }
  n2 = LinearInterpolate(A=VectorParameter() {parameterName=Base Color}.RGB, Alpha=@n1.R) {constB=0}
  n3 = ScalarParameter() {defaultValue=0.1, parameterName=Metallic}
  n4 = ScalarParameter() {defaultValue=1, parameterName=Specular}
  n5 = ScalarParameter() {defaultValue=0.1, parameterName=Roughness}
  n6 = TextureSampleParameter2D() {
   group=Mask Texture
   parameterName=BakedNormal
   samplerSource=SSM_Wrap_WorldGroupSettings
   samplerType=SAMPLERTYPE_Normal
   texture=/Game/Characters/Player/Female/Medium/Bodies/F_Med_Sheath/Textures/Alter_Ego/T_F_MED_Sheath_Alter_Ego_M_glass_N
  }
  n7 = Normalize(VectorInput=@n6.RGB)
  n8 = Fresnel(ExponentIn=ScalarParameter() {defaultValue=1, parameterName=Fresnel Exponent}, Normal=Transform(None=@n7))
  n9 = ScalarParameter() {defaultValue=1, parameterName=Glass thickness}
  n10 = LinearInterpolate(
   A=ScalarParameter() {defaultValue=0.9, parameterName=Fresnel Inner Transparency}
   B=ScalarParameter() {defaultValue=1, parameterName=Fresnel Outer Transparency}
   Alpha=Clamp(None=Multiply(A=@n8, B=@n9))
  )
  n11 = Multiply(A=LinearInterpolate(Alpha=@n1.R) {constA=0.8}, B=@n10)
  n12 = GetMaterialAttributes(None=MaterialFunctionCall!(
    Alpha=Constant() {r=1}
    Material=MakeMaterialAttributes!(
     BaseColor=@n2
     Metallic=@n3
     Specular=@n4
     Roughness=@n5
     Anisotropy
     EmissiveColor
     Opacity=@n11
     OpacityMask=@n10
     Normal=@n7
     Tangent
     WorldPositionOffset
     SubsurfaceColor
     ClearCoat
     ClearCoatRoughness
     AmbientOcclusion
     Refraction
     CustomizedUVs_0
     CustomizedUVs_1
     CustomizedUVs_2
     CustomizedUVs_3
     CustomizedUVs_4
     CustomizedUVs_5
     CustomizedUVs_6
     CustomizedUVs_7
     PixelDepthOffset
     ShadingModel
     Displacement
    )
   ) {materialFunction=/Game/Weapons/WeaponSkins/MaterialFunction/SubUVTextures}.Result) {attributeGetTypes [#1]}
  n13 = MaterialFunctionCall(UV Channel=TextureCoordinate()) {materialFunction=&LinearGradient}
  n14 = Multiply(A=TextureCoordinate(), B=ScalarParameter() {defaultValue=60, parameterName=asdfasdf})
  n15 = Saturate(None=MaterialFunctionCall(Seed=Time(), Vec2=Ceil(None=@n14)) {materialFunction=&MF_Hash}.Vec2)
  n16 = Multiply(
   A=Clamp(None=LinearInterpolate(Alpha=@n13.UGradient) {constB=2.5})
   B=LinearInterpolate(Alpha=ComponentMask(None=@n15) {g=true}) {constA=0.5}
  )
  n17 = Add(
   A=Subtract(A=Clamp(None=@n16) {maxDefault=0.99, minDefault=0.02})
   B=Max(A=ScalarParameter() {parameterName=GoggleTransition, sliderMax=2}, B=ScalarParameter() {parameterName=FrontEnd_Idle_Bypass})
  )
  n18 = Ceil(None=@n17)
  MaterialFunctionCall(In=SetMaterialAttributes(MaterialAttributes=@n12.MaterialAttributes, Emissive Color=Multiply(A=@n12.Emissive Color, B=@n18)) {attributeSetTypes [#1]}) {materialFunction=&MF_TechArtOverrides}
  Multiply(
   A=VectorParameter() {parameterName=Leading Edge Color, defaultValue=<rgba 4 0 3.5758 1>}.RGB
   B=SphereMask(A=@n17, B=ScalarParameter() {parameterName=Ring}, Radius=ScalarParameter() {parameterName=Radius})
  )
  Multiply()
  Multiply(B=@n18)
 ]
}
